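# Physics asset for the face accessory mesh of the F_MED_SacredCuddle female medium character body, setting up physics for its two ears.
# It has capsule bodies on the head and on the left and right dyn_ear01 bones.
# Each ear body is constrained to the head with matching angular swing and twist limits, twist being tighter than swing, and the ears have no constraint to each other.
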PhysicsAsset /BRCosmetics/Characters/Player/Female/Medium/Bodies/F_MED_SacredCuddle/Meshes/Parts/F_MED_SacredCuddle_FaceAcc_Physics
bodies {
 dyn_ear01_l [{
   shapeType=Capsule
   length=5.3989
   radius=4.0498
   shapeName=dyn_ear01_l_capsule
   center=<4.2101 0.0313 1.3707>
   rotation=<rot 0 3.7052 -87.7145>
  }]
 dyn_ear01_r [{
   shapeType=Capsule
   length=5.3332
   radius=4.9053
   shapeName=dyn_ear01_r_capsule
   center=<4.1306 0.3855 -0.1756>
   rotation=<rot 0 -167.3893 -87.7428>
  }]
 head [{shapeType=Capsule, length=10.5301, radius=9.2894, shapeName=head_capsule, center=<4.6917 0.6324 -0.2413>, rotation=<rot -0.237 1.7379 0.789>}]
}
constraints [
 {bone1Name=dyn_ear01_r, bone2Name=head, swing1LimitDegrees=45, swing2LimitDegrees=45, twistLimitDegrees=15}
 {bone1Name=dyn_ear01_l, bone2Name=head, swing1LimitDegrees=45, swing2LimitDegrees=45, twistLimitDegrees=15}
]
