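# Animation Blueprint for the Sacred Cuddle female cosmetic's face accessory, adding physics-driven secondary motion.
# AnimGraph copies the attached parent mesh's pose, modifies the head bone, then simulates RigidBody physics in spine_05 space with damped component motion.
# EventGraph computes gravity override, scaled inverse head-versus-neck_02 rotation, and a not-emoting flag each update.
# Those variables have no visible AnimGraph connection; the bone transform and gravity inputs show constants.
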AnimBlueprint /BRCosmetics/Characters/Player/Female/Medium/Bodies/F_MED_SacredCuddle/Meshes/Parts/F_MED_SacredCuddle_FaceAcc_AnimBP
target_skeleton=&/F_MED_SacredCuddle_FaceAcc_Skeleton
graphs [
 AnimGraph {nodes [
   n1 = "Transform (Modify) Bone - Bone: head"!(
    ComponentPose=Local To Component(LocalPose=Copy Pose From Mesh() {bUseAttachedParent=true})
    bAlphaBoolEnabled:bool=True
    Translation="0,0,0"
    Rotation="0,0,0"
    Scale="1,1,1"
   ) -> (Pose) {
    alphaInputType=Bool
    rotationMode=BMM_Replace
    rotationSpace=BCS_BoneSpace
    alphaBoolBlend {blendOption=Linear}
    boneToModify {boneName=head}
   }
   n2 = RigidBody!(ComponentPose=@n1, Alpha, OverrideWorldGravity:Vector="0,0,-980", ExternalForce) -> (Pose) {
    bClampLinearTranslationLimitToRefPose=true
    bOverrideWorldGravity=true
    overridePhysicsAsset=&/F_MED_SacredCuddle_FaceAcc_Physics
    simulationSpace=BaseBoneSpace
    alphaBoolBlend {blendOption=Linear}
    baseBoneRef {boneName=spine_05}
    componentLinearAccScale=<0.6 0.4 0.4>
    componentLinearVelScale=<0.2 0.2 0.2>
    overrideWorldGravity {z=-980}
    simSpaceSettings {maxAngularAcceleration=800, maxAngularVelocity=3, maxLinearAcceleration=400, maxLinearVelocity=250, velocityScaleZ=0, worldAlpha=1}
   }
   Output Pose(Result=Component To Local(ComponentPose=@n2)) {layerGroup=None}
  ]}
 EventGraph {nodes [
   Event BlueprintUpdateAnimation()
   > n1 = Gravity Override State Function DEPRECATED(CharacterPartAnimInstance=Self-Reference(), GameTimeInSeconds=GetGameTimeInSeconds(), GravityOverrideParams=Get Gravity Override Params())
   > Set Gravity Override(Gravity Override=@n1)
   n2 = GetOwnerSkeletalMeshComponent()
   n3 = BreakRotator(InRot=BreakTransform(InTransform=GetDeltaTransformFromRefPose(self=@n2, BoneName=head, BaseName=neck_02)).Rotation)
   n4 = Get Z_Percentage() -> (Z_Percentage:double)
   n5 = Get X_Percentage() -> (X_Percentage:double)
   n6 = Get Y_Percentage() -> (Y_Percentage:double)
   > Set HeadRot(HeadRot:Rotator=MakeRotator(
     Roll=float * float!(A=@n3, B=-1, C:double=@n5) -> (ReturnValue) {is=title K2Node_PromotableOperator}
     Pitch=float * float(A=@n3.Pitch, B=-1, C=@n6)
     Yaw=float * float(A=@n4, B=-1, C:float=@n3.Yaw)
    )) -> (Output_Get:Rotator)
   > Set IsNotPlayintEmote(IsNotPlayintEmote:bool=NOT Boolean(A=Get bIsPlayingEmote())) -> (Output_Get:bool)
  ]}
]
variables [
 Gravity Override {type=Vector}
 Gravity Override Params {type=GravityOverrideParamsStruct}
 HeadRot {type=Rotator}
 IsNotPlayintEmote {type=boolean}
 X_Percentage {type=number, default=1}
 Y_Percentage {type=number, default=0.6}
 Z_Percentage {type=number, default=0.6}
]
